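; YokeHopUnkind game state component that customises storm phases and drives storm material parameters.
; It listens for safe zone state and phase events, feeding shrink progress into a material parameter collection every frame during phase 1.
; ScalableFloat-gated server logic recentres phase 2, rerolls later locations and derives minimum shrink times from edge distance and expected speed.
; The Compactor rectangle prototype forces phase centres or caps their movement; a tick-driven shield reveal remains, marked unused.
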(event Custom|UpdateShrinkPhaseMaterialParams
  (bind _returnvalue (SafeZone|GetSafeZoneIndicator))
  (bind _returnvalue_2 (StormShape|GetShrinkPhaseCompletionAlpha))
  (Utilities|IsValid _returnvalue
    (:Is%20Valid
      (Rendering|Material|SetScalarParameterValue "/YokeHopUnkind/Assets/MPC_YokeHopUnkind.MPC_YokeHopUnkind" "ShrinkPhaseCompletionAlpha" _returnvalue_2 :!class /Script/Engine.KismetMaterialLibrary)
      (if (or (>= _returnvalue_2 1.0) (> (SafeZone|GetCurrentPhaseIndex _returnvalue) 1))
        (Rendering|Material|SetScalarParameterValue "/YokeHopUnkind/Assets/MPC_YokeHopUnkind.MPC_YokeHopUnkind" "ShrinkPhaseCompletionAlpha" 1.0 :!class /Script/Engine.KismetMaterialLibrary)
        (Utilities|Time|ClearandInvalidateTimerbyHandle (Variables|Default|GetShrinkPhaseUpdateTimer))
        (else
          _)))
    (:Is%20Not%20Valid
      _))
  (bind _returnvalue_3 (Utilities|Time|SetTimerforNextTickbyEvent (delegate Custom|UpdateShrinkPhaseMaterialParams)))
  (Variables|Default|SetShrinkPhaseUpdateTimer _returnvalue_3))

(event Custom|DebugTinyStorm
  (bind _returnvalue (SafeZone|GetSafeZoneIndicator))
  (bind _returnvalue_2 (SafeZone|GetCurrentPhaseIndex _returnvalue))
  (bind _returnvalue_3 (+ _returnvalue_2 1))
  (SafeZone|ModifyPhase _returnvalue _returnvalue_3 (Utilities|Struct|MakeFortSafeZonePhaseInfo "0.000000,0.000000,50.000000" 6e+03 2e+01 2e+01))
  (Utilities|FlowControl|Delay 0.2)
  (SafeZone|StartPhase _returnvalue _returnvalue_3))

(event EventBeginPlay
  (Utilities|FlowControl|Sequence
    (:then_0
      (bind _returnvalue (GameplayEventRouter|GetEventRouter "/Script/FortniteGame.FortGameState" (Components|GetOwner)))
      (bind (_asynctaskproxy _payload) (GameplayEventRouter|StartListeningToStatefulEvent _returnvalue "/Script/FortniteGame.FortSafeZoneStateUpdatedEvent")
        (:then
          ;[1 1 1] If we change phase while the storm is already holding, we won't get that state update again, so this is here to catch those cases for easier testing
          (bind (_asynctaskproxy_2 _payload_2) (GameplayEventRouter|StartListeningToStatefulEvent _returnvalue "/Script/FortniteGame.FortSafeZonePhaseUpdatedEvent")
            (:OnEventReceived
              (Development|PrintString "New Phase")
              (CallFunction|StartReveal)
              (if (== (Utilities|Struct|BreakFortSafeZonePhaseUpdatedEvent _payload_2) 1)
                (Utilities|FlowControl|Sequence
                  (:then_0
                    (bind _returnvalue_2 (Utilities|Time|SetTimerforNextTickbyEvent (delegate Custom|UpdateShrinkPhaseMaterialParams)))
                    (Variables|Default|SetShrinkPhaseUpdateTimer _returnvalue_2))
                  (:then_1
                    (if (or (ScalableFloat|GetValueasBool (Variables|Default|GetHF_RectangleProtoEnabled)) (Variables|Default|GetConsoleRectangleProtoEnabled))
                      (CallFunction|SetUpRectangleProtoPhases))))))))
        (:OnEventReceived
          (switch Utilities|FlowControl|Switch|SwitchonEFortSafeZoneState (.CurrentState (Utilities|Struct|BreakFortSafeZoneStateUpdatedEvent _payload))
            (:Starting
              (Development|PrintString "Starting")
              (CallFunction|StartReveal)
              (Networking|SwitchHasAuthority
                (:Authority
                  ;[1 1 1] Don't like this but the SafeZoneIndicator appears to not be valid immediately after this event fires, and not sure what the right callback is
                  (Utilities|FlowControl|Delay 0.2)
                  (Variables|Default|SetbStormStarted true)
                  (Utilities|FlowControl|Sequence
                    (:then_0
                      (if (ScalableFloat|GetValueasBool (Variables|Default|GetHF_PutPhase2AtOrigin))
                        (CallFunction|CenterPhase2)))
                    (:then_1
                      (if (ScalableFloat|GetValueasBool (Variables|Default|GetHF_EnforceShrinkTimes))
                        (CallFunction|EnforceShrinkTimes false) ; Skip ShrinkTimes on Phase 0+1 to start\n(should use WaveInfill data for these)
                      ))))))
            (:Holding
              (Development|PrintString "Holding")
              (CallFunction|StartReveal))))))))

(event EventOnRepeatedPhasesReset
  (Networking|SwitchHasAuthority
    (:Authority
      (if (ScalableFloat|GetValueasBool (Variables|Default|GetHF_EnforceShrinkTimes))
        (CallFunction|EnforceShrinkTimes true)))))

(event Custom|EnableRectangleStormProto
  (Variables|Default|SetConsoleRectangleProtoEnabled true)
  (CallFunction|MulticastEnableRectangleStormProto)
  (if (Variables|Default|GetbStormStarted)
    (CallFunction|SetUpRectangleProtoPhases)))

(event Custom|MulticastEnableRectangleStormProto
  (bind _returnvalue (Mutator|GetGameplayMutatorbyClass "/YokeHopUnkind/Mutators/BP_YokeHopUnkindMutator.BP_YokeHopUnkindMutator_C"))
  (Utilities|IsValid _returnvalue
    (:Is%20Valid
      (Class|BPYokeHopUnkindMutator|EnableRectangleStormProto _returnvalue))))

(event Custom|StartReveal
  ;[0.15 6.4e-05 0.008305 0.5] Reveal is no longer used. Do Not tick
  (bind _returnvalue (Development|GetConsoleVariableBoolValue "StormShape.UseStormShapeTickInsteadOfTimers"))
  (if (not _returnvalue)
    (Networking|CanExecuteCosmeticEvents
      (:True
        (if (not (Variables|Default|GetbStartedReveal))
          (Variables|Default|SetbStartedReveal true)
          (Variables|Default|SetRevealStartTime (Utilities|Time|GetTimeSeconds))
          (Components|Tick|SetComponentTickEnabled true))))))

(event Custom|UpdateReveal
  (bind _returnvalue (Utilities|Time|GetTimeSeconds))
  (bind _revealstarttime (Variables|Default|GetRevealStartTime))
  (bind _returnvalue_2 (- _returnvalue _revealstarttime))
  (bind _revealduration (Variables|Default|GetRevealDuration))
  ;[0.15 6.4e-05 0.008305 0.5] Reveal is no longer used. Do Not tick
  (bind _returnvalue_3 (Development|GetConsoleVariableBoolValue "StormShape.UseStormShapeTickInsteadOfTimers"))
  (if _returnvalue_3
    _
    (elif (> _returnvalue_2 _revealduration)
      (Rendering|Material|SetScalarParameterValue "/YokeHopUnkind/Assets/MPC_YokeHopUnkind.MPC_YokeHopUnkind" "StormShieldReveal" 1.0 :!class /Script/Engine.KismetMaterialLibrary)
      (Components|Tick|SetComponentTickEnabled _returnvalue_3)
      (else
        (Rendering|Material|SetScalarParameterValue "/YokeHopUnkind/Assets/MPC_YokeHopUnkind.MPC_YokeHopUnkind" "StormShieldReveal" (Math|Float|Clamp(Float) (/ _returnvalue_2 _revealduration)) :!class /Script/Engine.KismetMaterialLibrary)))))

(event Custom|DestroySafeZone
  (StormShape|HandleSafeZoneDestroyed))

(event Custom|ResetStormReveal
  ;[1 1 1] Cheat event to reset reveal flag, so it can be seen multiple times in one play
  (Variables|Default|SetbStartedReveal false))

(event StormShape|EventBP_OnStormShapeSelected
  (params (CurrentStormPhase :type Integer) (ShapeConfigIdTag :type Gameplay%20Tag%20Structure) (ShapeProviderSetIndex :type Integer) (ShapeProviderIndex :type Integer))
  (if (and (== CurrentStormPhase 1) (GameplayTags|MatchesTag ShapeConfigIdTag (Variables|Default|GetCompactorStormTag) true))
    (Variables|Default|SetConsoleRectangleProtoEnabled true)))

(event EventTick
  (params (DeltaSeconds :type Float%20%28single-precision%29))
  ;[0.15 6.4e-05 0.008305 0.5] Reveal is no longer used. Do Not tick
  (bind _returnvalue (Development|GetConsoleVariableBoolValue "StormShape.UseStormShapeTickInsteadOfTimers"))
  (if (not _returnvalue)
    (Networking|CanExecuteCosmeticEvents
      (:True
        (CallFunction|UpdateReveal))
      (:False
        (bind _returnvalue_2 (Development|GetConsoleVariableBoolValue "StormShape.UseStormShapeTickInsteadOfTimers"))
        (if (not _returnvalue_2)
          (Components|Tick|SetComponentTickEnabled false))))))

(event StormShape|EventBP_HandleNewMovementState
  (params (NewMovementState :type EStormMovementState%20Enum))
  (Default|CallOnStormMovementStateChanged NewMovementState))

(event StormShape|EventBP_OnSplineMeshCreated
  (params (NewSplineMesh :type Spline%20Mesh%20Component%20Object%20Reference))
  (Rendering|SetTranslucentSortPriority NewSplineMesh 0))

(event StormShape|EventBP_OnStormShapeRegistryLoaded
  (Variables|Default|SetbRegistryLoaded true)
  (Development|PrintString "Registry loaded")
  (Default|CallOnStormShapeRegistryLoaded))

(fn CenterPhase2
  (bind _returnvalue (SafeZone|GetSafeZoneIndicator))
  (bind _returnvalue_2 (BattleRoyaleGamePhaseLogic|GetBattleRoyaleGamePhaseLogicComponent))
  (bind _returnvalue_3 (SafeZone|GetSafeZoneIndicator))
  (bind _structout (Utilities|Struct|SetmembersinFortSafeZonePhaseInfo (SafeZone|GetPhaseInfo _returnvalue_3 2) "0.000000,0.000000,0.000000"))
  (SafeZone|ModifyPhase _returnvalue_3 2 _structout)
  (if (ScalableFloat|GetValueasBool (Variables|Default|GetHF_RerollPhases))
    (bind _output_get (LocalVariables|SetNewLocations (SafeZones|GenerateSafeZoneLocations (Variables|/Script/FortniteGame.FortGameStateAthena|GetMapInfo (Fortnite|Athena|GameState|GetGameStateAthena)) (BattleRoyaleGamePhaseLogic|GetSafeZoneLocationDenyListArray _returnvalue_2) 3)))
    (BattleRoyaleGamePhaseLogic|SetSafeZoneLocations _returnvalue_2 _output_get)
    (for _index (range 3 (Utilities|Array|Length (LocalVariables|GetNewLocations)))
      (bind _structout_2 (Utilities|Struct|SetmembersinFortSafeZonePhaseInfo (SafeZone|GetPhaseInfo _returnvalue _index) (Utilities|Array|Get(acopy) (LocalVariables|GetNewLocations) _index)))
      (SafeZone|ModifyPhase _returnvalue _index _structout_2))))

(fn EnforceShrinkTimes
  (params (bUpdateFirstTwoPhases :type Boolean))
  (bind _currentphaseinfo (LocalVariables|GetCurrentPhaseInfo))
  (bind (_center _radius _waittime _shrinktime) (Utilities|Struct|BreakFortSafeZonePhaseInfo _currentphaseinfo))
  (bind _returnvalue (SafeZone|GetSafeZoneIndicator))
  (bind _returnvalue_2 (SafeZone|GetPhaseCount _returnvalue))
  (bind _minshrinktime (LocalVariables|GetMinShrinkTime))
  (bind _edgedistance (LocalVariables|GetEdgeDistance))
  (bind _currentphase (LocalVariables|GetCurrentPhase))
  (bind _prevphase (LocalVariables|GetPrevPhase))
  (bind _result (Utilities|Text|FormatText "NSLOCTEXT(\"[7C769A8F75F122997D7EC5E5B9C74A0D]\", \"9B6B5E0D40B6787A6D2CD9A312EF9C03\", \"Phase {PrevPhase} -> Phase {CurrPhase}: Distance = {Dist}, MinShrinkTime = {MinShrinkTime}, Actual Shrink Time = {ShrinkTime}\")" _prevphase _currentphase _edgedistance _minshrinktime (Math|Float|Max(Float) _shrinktime (LocalVariables|GetMinShrinkTime))))
  (bind _returnvalue_3 (Utilities|String|ToString(Text) _result))
  (bind _returnvalue_4 (SafeZone|GetSafeZoneIndicator))
  (bind _minshrinktime_2 (LocalVariables|GetMinShrinkTime))
  (bind _returnvalue_5 (Math|Float|Max(Float) _shrinktime _minshrinktime_2))
  (Utilities|IsValid _returnvalue
    (:Is%20Valid
      (if (> _returnvalue_2 0)
        (for _index (range (Math|Integer|SelectInt 0 2 (Variables|GetbUpdateFirstTwoPhases)) (+ (Math|Integer|SelectInt 2 (- _returnvalue_2 1) (Variables|GetbUpdateFirstTwoPhases)) 1))
          (LocalVariables|SetCurrentPhase _index)
          (LocalVariables|SetPrevPhase (+ _index -1))
          (if (and (.ReturnValue (SafeZone|GetPhaseInfo _returnvalue_4 _index)) (.ReturnValue (SafeZone|GetPhaseInfo _returnvalue_4 (+ _index -1))))
            (LocalVariables|SetCurrentPhaseInfo (SafeZone|GetPhaseInfo _returnvalue_4 _index))
            (bind _output_get (LocalVariables|SetEdgeDistance (+ (Math|Vector|Distance(Vector) (Utilities|Struct|BreakFortSafeZonePhaseInfo (SafeZone|GetPhaseInfo _returnvalue_4 _index)) (Utilities|Struct|BreakFortSafeZonePhaseInfo (SafeZone|GetPhaseInfo _returnvalue_4 (+ _index -1)))) (- (.Radius (Utilities|Struct|BreakFortSafeZonePhaseInfo (SafeZone|GetPhaseInfo _returnvalue_4 (+ _index -1)))) (.Radius (Utilities|Struct|BreakFortSafeZonePhaseInfo (SafeZone|GetPhaseInfo _returnvalue_4 _index)))))))
            (LocalVariables|SetMinShrinkTime (/ _output_get (ScalableFloat|GetValueatLevel (Variables|Default|GetHF_ExpectedSpeed))))
            (bind _structout (Utilities|Struct|SetmembersinFortSafeZonePhaseInfo _currentphaseinfo _returnvalue_5))
            (SafeZone|ModifyPhase (SafeZone|GetSafeZoneIndicator) (LocalVariables|GetCurrentPhase) _structout)
            (Development|PrintString _returnvalue_3 true true "(R=0.000000,G=0.660000,B=1.000000,A=1.000000)" 2e+01)
            (Fortnite|FortShippingLog self _returnvalue_3)))))))

(fn SetupPostCompactorPhase
  (params (LastCompactorPhase :type Integer))
  (bind _returnvalue (+ LastCompactorPhase 1))
  (bind _returnvalue_2 (SafeZone|GetSafeZoneIndicator))
  (bind _outsafezonephase (SafeZone|GetPhaseInfo _returnvalue_2 _returnvalue))
  (bind _center (Utilities|Struct|BreakFortSafeZonePhaseInfo _outsafezonephase))
  (bind _returnvalue_3 (Math|Conversions|ToFloat(Integer) _returnvalue))
  (bind _hf_compactormovedistance (Variables|Default|GetHF_CompactorMoveDistance))
  (bind _returnvalue_4 (ScalableFloat|GetValueatLevel _hf_compactormovedistance _returnvalue_3))
  (bind _output_get (LocalVariables|SetPrevCenter (Utilities|Struct|BreakFortSafeZonePhaseInfo (SafeZone|GetPhaseInfo _returnvalue_2 LastCompactorPhase))))
  (if (> (Math|Vector|Distance2D(Vector) _output_get _center) _returnvalue_4)
    (bind _structout (Utilities|Struct|SetmembersinFortSafeZonePhaseInfo _outsafezonephase (+ (* (Math|Vector|GetUnitDirection(Vector) (LocalVariables|GetPrevCenter) _center) _returnvalue_4) (LocalVariables|GetPrevCenter))))
    (SafeZone|ModifyPhase (SafeZone|GetSafeZoneIndicator) _returnvalue _structout)))

(fn SetUpRectangleProtoPhases
  (bind _returnvalue (SafeZone|GetSafeZoneIndicator))
  (if (ScalableFloat|GetValueasBool (Variables|Default|GetHF_CompactorForceCentered))
    (LocalVariables|SetPhase1Center (Utilities|Struct|BreakFortSafeZonePhaseInfo (SafeZone|GetPhaseInfo (SafeZone|GetSafeZoneIndicator) 1)))
    (for _index (range 1 13)
      (if (.ReturnValue (SafeZone|GetPhaseInfo _returnvalue _index))
        (bind _structout (Utilities|Struct|SetmembersinFortSafeZonePhaseInfo (SafeZone|GetPhaseInfo _returnvalue _index) (Math|Vector|SelectVector (LocalVariables|GetPhase1Center) "0, 0, 0" (ScalableFloat|GetValueasBool (Variables|Default|GetHF_CompactorUseFirstPhaseAsForcedCenter)))))
        (SafeZone|ModifyPhase _returnvalue _index _structout)))
    (else
      (LocalVariables|SetLastPhase (ScalableFloat|GetValueasInteger (Variables|Default|GetHF_LastCompactorPhase)))
      (CallFunction|SetupPostCompactorPhase (LocalVariables|GetLastPhase))
      (for _index_2 (range (+ (LocalVariables|GetLastPhase) 1) 13)
        (CallFunction|SetupPostCompactorPhase _index_2)))))
